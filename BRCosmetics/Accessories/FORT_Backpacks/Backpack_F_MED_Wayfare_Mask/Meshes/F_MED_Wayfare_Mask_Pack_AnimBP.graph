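# Animation Blueprint for the Wayfare Mask backpack cosmetic, animating a hovering pack alongside the player's movement.
# Locomotion and DBNO state machines save cached poses, with idle, directional run, jump and hunker states reading player anim state.
# The Gliding machine's pose gets additive pack and hover rotations, and hover scales to zero in vehicles.
# The final pose blends the pack bone between an additive carrier offset and a world-space target by smoothed alpha.
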AnimBlueprint /BRCosmetics/Accessories/FORT_Backpacks/Backpack_F_MED_Wayfare_Mask/Meshes/F_MED_Wayfare_Mask_Pack_AnimBP
target_skeleton=&/F_MED_Wayfare_Mask_Pack_Skeleton
graphs [
 AnimGraph {nodes [
   n1 = "Transform (Modify) Bone - Bone: pack"(
    ComponentPose=Local To Component(LocalPose=Use cached pose 'PreSpringPose'())
    Translation=Get BackpackTargetVector() -> (BackpackTargetVector:Vector)
    Rotation="0,0,0"
   ) {translationMode=BMM_Replace, translationSpace=BCS_WorldSpace, alphaBoolBlend {blendOption=Linear}, boneToModify {boneName=pack}}
   n2 = "Transform (Modify) Bone - Bone: pack"(
    ComponentPose=Local To Component(LocalPose=Use cached pose 'PreSpringPose'())
    Translation=Get CarrierOffsetTotal() -> (CarrierOffsetTotal:Vector)
    Rotation="0,0,0"
   ) {translationMode=BMM_Additive, alphaBoolBlend {blendOption=Linear}, boneToModify {boneName=pack}}
   Output Pose(Result=Two Way Blend(
     A=Component To Local(ComponentPose=@n2)
     B=Component To Local(ComponentPose=@n1)
     Alpha=Get BackpackTargetVectorAlphaSmooth() -> (BackpackTargetVectorAlphaSmooth:float)
    ) {blendNode {alphaBoolBlend {blendOption=Linear}}}) {layerGroup=None}
   AnimGraphNode_StateMachine_1 = Gliding() {is=title AnimGraphNode_StateMachine}
   n3 = "Transform (Modify) Bone - Bone: hover"(
    ComponentPose="Transform (Modify) Bone - Bone: pack"(ComponentPose=Local To Component(LocalPose=@AnimGraphNode_StateMachine_1), Translation="0,0,0", Rotation="0,180,0") {rotationMode=BMM_Additive, alphaBoolBlend {blendOption=Linear}, boneToModify {boneName=pack}, rotation {yaw=180}}
    Alpha=1
    Translation="0,0,0"
    Rotation=Get HoverRotationOffsetTotal() -> (HoverRotationOffsetTotal:Rotator)
    Scale="1,1,1"
   ) {rotationMode=BMM_Additive, alphaBoolBlend {blendOption=Linear}, boneToModify {boneName=hover}}
   n4 = Get CarrierOffsetRotationVelocity() -> (CarrierOffsetRotationVelocity:Rotator)
   n5 = Get CarrierOffsetRotationVelocityAlpha() -> (CarrierOffsetRotationVelocityAlpha:float)
   n6 = "Transform (Modify) Bone - Bone: hover"(
    ComponentPose="Transform (Modify) Bone - Bone: pack"(
     ComponentPose="Transform (Modify) Bone - Bone: hover"(ComponentPose=@n3, Alpha=@n5, Translation="0,0,0", Rotation=@n4, Scale="1,1,1") {rotationMode=BMM_Additive, alphaBoolBlend {blendOption=Linear}, boneToModify {boneName=hover}}
     Translation="0,0,0"
     Rotation=Get RotationYawDefinedHoverOffset() -> (RotationYawDefinedHoverOffset:Rotator)
    ) {rotationMode=BMM_Additive, alphaBoolBlend {blendOption=Linear}, boneToModify {boneName=pack}}
    Alpha="To Float (Boolean)"(InBool=Get bIsInVehicle() -> (bIsInVehicle:bool))
    Translation="0,0,0"
    Rotation="0,0,0"
    Scale="0,0,0"
   ) {scaleMode=BMM_Replace, alphaBoolBlend {blendOption=Linear}, boneToModify {boneName=hover}, scale=<0 0 0>}
   Save cached pose 'PreSpringPose'(Pose=Component To Local(ComponentPose=@n6)) {cacheName=PreSpringPose}
   AnimGraphNode_StateMachine_0 = Locomotion() {is=title AnimGraphNode_StateMachine}
   Save cached pose 'LocoCache'(Pose=@AnimGraphNode_StateMachine_0) {cacheName=LocoCache}
   AnimGraphNode_StateMachine_2 = DBNO() {is=title AnimGraphNode_StateMachine}
   Save cached pose 'FullLocoCache'(Pose=@AnimGraphNode_StateMachine_2) {cacheName=FullLocoCache}
  ]}
 Locomotion {
  entry=Locomotion
  path=AnimGraph.AnimGraphNode_StateMachine_0.Locomotion
  ids=id
  nodes [
   AnimStateConduitNode_0 = JumpConduit() {is=title AnimStateConduitNode}
   AnimStateNode_0 = JumpIntro() {is=title AnimStateNode}
   AnimStateNode_1 = JumpLoop() {is=title AnimStateNode}
   AnimStateNode_10 = Hunker() {
    is=title AnimStateNode
    stateEntered {guid=#1}
    stateFullyBlended {guid=#2}
    stateLeft {guid=#3}
   }
   AnimStateNode_11 = HunkerOutro() {
    is=title AnimStateNode
    stateEntered {guid=#4}
    stateFullyBlended {guid=#5}
    stateLeft {guid=#6}
   }
   AnimStateNode_12 = JumpOutroRun() {is=title AnimStateNode, stateEntered {guid=#7, notifyName=Entered_JumpOutroRun}}
   AnimStateNode_13 = Idle()
   AnimStateNode_14 = IntroRun() {is=title AnimStateNode}
   AnimStateNode_15 = Locomotion Placeholder To Save Transitions() {is=title AnimStateNode}
   AnimStateNode_2 = JumpOutro() {is=title AnimStateNode, stateEntered {guid=#8, notifyName=Entered_JumpOutro}}
   AnimStateNode_3 = OutroRun() {is=title AnimStateNode}
   AnimStateNode_4 = SkyDiving() {is=title AnimStateNode}
   AnimStateNode_5 = SkyDiveOutro() {is=title AnimStateNode}
   AnimStateNode_6 = SkyDivingIntro() {is=title AnimStateNode}
   AnimStateNode_7 = Jump() {is=title AnimStateNode}
   AnimStateNode_8 = RunLoop() {is=title AnimStateNode}
   AnimStateNode_9 = Locomotion() {
    is=title AnimStateNode
    stateEntered {guid=#9}
    stateFullyBlended {guid=#10}
    stateLeft {guid=#11}
   }
   AnimStateTransitionNode_0 = AnimStateTransitionNode()
   AnimStateTransitionNode_10 = AnimStateTransitionNode() {crossfadeDuration=0.4}
   AnimStateTransitionNode_11 = AnimStateTransitionNode() {bAutomaticRuleBasedOnSequencePlayerInState=true, crossfadeDuration=0.45}
   AnimStateTransitionNode_13 = AnimStateTransitionNode() {bAutomaticRuleBasedOnSequencePlayerInState=true}
   AnimStateTransitionNode_14 = AnimStateTransitionNode()
   AnimStateTransitionNode_16 = AnimStateTransitionNode()
   AnimStateTransitionNode_17 = AnimStateTransitionNode()
   AnimStateTransitionNode_18 = AnimStateTransitionNode()
   AnimStateTransitionNode_19 = AnimStateTransitionNode()
   AnimStateTransitionNode_2 = AnimStateTransitionNode() {
    crossfadeDuration=0.4
    transitionEnd {guid=#12}
    transitionInterrupt {guid=#13}
    transitionStart {guid=#14}
   }
   AnimStateTransitionNode_20 = AnimStateTransitionNode()
   AnimStateTransitionNode_21 = AnimStateTransitionNode()
   AnimStateTransitionNode_25 = AnimStateTransitionNode() {bAutomaticRuleBasedOnSequencePlayerInState=true, crossfadeDuration=0.8}
   AnimStateTransitionNode_3 = AnimStateTransitionNode() {bAutomaticRuleBasedOnSequencePlayerInState=true, crossfadeDuration=0.73}
   AnimStateTransitionNode_4 = AnimStateTransitionNode() {
    bAutomaticRuleBasedOnSequencePlayerInState=true
    crossfadeDuration=0.7
    transitionEnd {guid=#15}
    transitionInterrupt {guid=#16}
    transitionStart {guid=#17}
   }
   AnimStateTransitionNode_5 = AnimStateTransitionNode() {
    crossfadeDuration=0.3
    transitionEnd {guid=#18}
    transitionInterrupt {guid=#19}
    transitionStart {guid=#20}
   }
   AnimStateTransitionNode_6 = AnimStateTransitionNode()
   AnimStateTransitionNode_7 = AnimStateTransitionNode() {crossfadeDuration=0.3}
   AnimStateTransitionNode_8 = AnimStateTransitionNode() {crossfadeDuration=0.75}
   AnimStateTransitionNode_9 = AnimStateTransitionNode() {
    bAutomaticRuleBasedOnSequencePlayerInState=true
    crossfadeDuration=0.9
    transitionEnd {guid=#21}
    transitionInterrupt {guid=#22}
    transitionStart {guid=#23}
   }
  ]
  links [
   {
    dst=HunkerOutro
    rule=AnimGraph.AnimGraphNode_StateMachine_0.Locomotion.AnimStateTransitionNode_2.AnimationTransitionGraph_1
    src=Hunker
    via=AnimStateTransitionNode_2
   }
   {
    dst=Locomotion
    rule=AnimGraph.AnimGraphNode_StateMachine_0.Locomotion.AnimStateTransitionNode_4.AnimationTransitionGraph_1
    src=HunkerOutro
    via=AnimStateTransitionNode_4
   }
   {
    dst=IntroRun
    rule=AnimGraph.AnimGraphNode_StateMachine_0.Locomotion.AnimStateTransitionNode_0.AnimationTransitionGraph_0
    src=Idle
    via=AnimStateTransitionNode_0
   }
   {
    dst=OutroRun
    rule=AnimGraph.AnimGraphNode_StateMachine_0.Locomotion.AnimStateTransitionNode_14.AnimationTransitionGraph_1
    src=IntroRun
    via=AnimStateTransitionNode_14
   }
   {
    dst=JumpConduit
    rule=AnimGraph.AnimGraphNode_StateMachine_0.Locomotion.AnimStateTransitionNode_6.Transition
    src=Jump
    via=AnimStateTransitionNode_6
   }
   {
    dst=Locomotion Placeholder To Save Transitions
    rule=AnimGraph.AnimGraphNode_StateMachine_0.Locomotion.AnimStateTransitionNode_20.Transition
    src=Jump
    via=AnimStateTransitionNode_20
   }
   {
    dst=JumpOutro
    rule=AnimGraph.AnimGraphNode_StateMachine_0.Locomotion.AnimStateTransitionNode_7.Transition
    src=JumpConduit
    via=AnimStateTransitionNode_7
   }
   {
    dst=JumpOutroRun
    rule=AnimGraph.AnimGraphNode_StateMachine_0.Locomotion.AnimStateTransitionNode_10.Transition
    src=JumpConduit
    via=AnimStateTransitionNode_10
   }
   {
    dst=JumpLoop
    rule=AnimGraph.AnimGraphNode_StateMachine_0.Locomotion.AnimStateTransitionNode_13.Transition
    src=JumpIntro
    via=AnimStateTransitionNode_13
   }
   {
    dst=JumpLoop
    rule=AnimGraph.AnimGraphNode_StateMachine_0.Locomotion.AnimStateTransitionNode_21.Transition
    src=JumpIntro
    via=AnimStateTransitionNode_21
   }
   {
    dst=JumpConduit
    rule=AnimGraph.AnimGraphNode_StateMachine_0.Locomotion.AnimStateTransitionNode_18.Transition
    src=JumpLoop
    via=AnimStateTransitionNode_18
   }
   {
    dst=Locomotion
    rule=AnimGraph.AnimGraphNode_StateMachine_0.Locomotion.AnimStateTransitionNode_25.Transition
    src=JumpOutro
    via=AnimStateTransitionNode_25
   }
   {
    dst=Locomotion
    rule=AnimGraph.AnimGraphNode_StateMachine_0.Locomotion.AnimStateTransitionNode_11.Transition
    src=JumpOutroRun
    via=AnimStateTransitionNode_11
   }
   {
    dst=Hunker
    rule=AnimGraph.AnimGraphNode_StateMachine_0.Locomotion.AnimStateTransitionNode_5.AnimationTransitionGraph_0
    src=Locomotion
    via=AnimStateTransitionNode_5
   }
   {
    dst=JumpIntro
    rule=AnimGraph.AnimGraphNode_StateMachine_0.Locomotion.AnimStateTransitionNode_16.Transition
    src=Locomotion
    via=AnimStateTransitionNode_16
   }
   {
    dst=JumpLoop
    rule=AnimGraph.AnimGraphNode_StateMachine_0.Locomotion.AnimStateTransitionNode_17.Transition
    src=Locomotion
    via=AnimStateTransitionNode_17
   }
   {
    dst=Jump
    rule=AnimGraph.AnimGraphNode_StateMachine_0.Locomotion.AnimStateTransitionNode_19.Transition
    src=Locomotion Placeholder To Save Transitions
    via=AnimStateTransitionNode_19
   }
   {
    dst=Idle
    rule=AnimGraph.AnimGraphNode_StateMachine_0.Locomotion.AnimStateTransitionNode_3.AnimationTransitionGraph_0
    src=OutroRun
    via=AnimStateTransitionNode_3
   }
   {
    dst=RunLoop
    rule=AnimGraph.AnimGraphNode_StateMachine_0.Locomotion.AnimStateTransitionNode_9.Transition
    src=SkyDiveOutro
    via=AnimStateTransitionNode_9
   }
   {
    dst=SkyDiveOutro
    rule=AnimGraph.AnimGraphNode_StateMachine_0.Locomotion.AnimStateTransitionNode_8.Transition
    src=SkyDiving
    via=AnimStateTransitionNode_8
   }
  ]
 }
 JumpConduit {path=AnimGraph.AnimGraphNode_StateMachine_0.Locomotion.AnimStateConduitNode_0.JumpConduit, nodes [Result(bCanEnterTransition=true)]}
 JumpIntro {
  path=AnimGraph.AnimGraphNode_StateMachine_0.Locomotion.AnimStateNode_0.JumpIntro
  nodes [
   n1 = Sequence Player 'Wayfare_Mask_JumpInto_PACK'!(bLoopAnimation:bool=False) -> (Pose) {bLoopAnimation=false, sequence=/BRCosmetics/Accessories/FORT_Backpacks/Backpack_F_MED_Wayfare_Mask/Animations/Wayfare_Mask_JumpInto_PACK}
   Output Animation Pose(Result=@n1)
  ]
 }
 JumpLoop {
  path=AnimGraph.AnimGraphNode_StateMachine_0.Locomotion.AnimStateNode_1.JumpLoop
  nodes [Output Animation Pose(Result=Sequence Player 'Wayfare_Mask_JumpLoop_PACK'() {sequence=/BRCosmetics/Accessories/FORT_Backpacks/Backpack_F_MED_Wayfare_Mask/Animations/Wayfare_Mask_JumpLoop_PACK})]
 }
 Hunker {
  path=AnimGraph.AnimGraphNode_StateMachine_0.Locomotion.AnimStateNode_10.Hunker
  nodes [Output Animation Pose(Result=Sequence Player 'Wayfare_Mask_Hunker_PET'() {sequence=/BRCosmetics/Accessories/FORT_Backpacks/Backpack_F_MED_Wayfare_Mask/Animations/Wayfare_Mask_Hunker_PET})]
 }
 HunkerOutro {
  path=AnimGraph.AnimGraphNode_StateMachine_0.Locomotion.AnimStateNode_11.HunkerOutro
  nodes [Output Animation Pose(Result=Sequence Player 'Wayfare_Mask_Hunker_Outro_PET'() {sequence=/BRCosmetics/Accessories/FORT_Backpacks/Backpack_F_MED_Wayfare_Mask/Animations/Wayfare_Mask_Hunker_Outro_PET})]
 }
 JumpOutroRun {
  path=AnimGraph.AnimGraphNode_StateMachine_0.Locomotion.AnimStateNode_12.JumpOutroRun
  nodes [Output Animation Pose(Result=Sequence Player 'Wayfare_Mask_JumpToRun_PET'() {sequence=/BRCosmetics/Accessories/FORT_Backpacks/Backpack_F_MED_Wayfare_Mask/Animations/Wayfare_Mask_JumpToRun_PET})]
 }
 Idle {
  path=AnimGraph.AnimGraphNode_StateMachine_0.Locomotion.AnimStateNode_13.Idle
  nodes [Output Animation Pose(Result=Sequence Player 'Wayfare_Mask_Idle'() {sequence=/BRCosmetics/Accessories/FORT_Backpacks/Backpack_F_MED_Wayfare_Mask/Animations/Wayfare_Mask_Idle})]
 }
 IntroRun {
  path=AnimGraph.AnimGraphNode_StateMachine_0.Locomotion.AnimStateNode_14.IntroRun
  nodes [
   Sequence Player 'Wayfare_Mask_TravelRun_Intro_Back'() {sequence=/BRCosmetics/Accessories/FORT_Backpacks/Backpack_F_MED_Wayfare_Mask/Animations/Wayfare_Mask_TravelRun_Intro_Back}
   Sequence Player 'Wayfare_Mask_TravelRun_Into_FWD'() {
    bLoopAnimation=false
    sequence=/BRCosmetics/Accessories/FORT_Backpacks/Backpack_F_MED_Wayfare_Mask/Animations/Wayfare_Mask_TravelRun_Into_FWD
   }
   Sequence Player 'Wayfare_Mask_TravelRun_Intro_Left'() {sequence=/BRCosmetics/Accessories/FORT_Backpacks/Backpack_F_MED_Wayfare_Mask/Animations/Wayfare_Mask_TravelRun_Intro_Left}
   Sequence Player 'Wayfare_Mask_TravelRun_Intro_Right'() {sequence=/BRCosmetics/Accessories/FORT_Backpacks/Backpack_F_MED_Wayfare_Mask/Animations/Wayfare_Mask_TravelRun_Intro_Right}
   AnimGraphNode_StateMachine_0 = New State Machine() {is=title AnimGraphNode_StateMachine}
   Output Animation Pose(Result=@AnimGraphNode_StateMachine_0)
   Get As Player Pawn Athena() -> (As Player Pawn Athena:Player Pawn Athena Object Reference)
   Get LocomotionCardinalDirection!(self:Fort Player Anim Instance Object Reference=Get As Fortnite M Avg Player Anim Blueprint() -> (As Fortnite M Avg Player Anim Blueprint:Fortnite M Avg Player Anim Blueprint Object Reference)) -> (LocomotionCardinalDirection:EFortCardinalDirection Enum=North)
  ]
 }
 New State Machine {
  entry=State
  path=AnimGraph.AnimGraphNode_StateMachine_0.Locomotion.AnimStateNode_14.IntroRun.AnimGraphNode_StateMachine_0.New State Machine
  ids=id
  nodes [
   AnimStateNode_0 = State() {is=title AnimStateNode}
   AnimStateNode_1 = North() {is=title AnimStateNode}
   AnimStateNode_2 = South() {is=title AnimStateNode}
   AnimStateNode_3 = East() {is=title AnimStateNode}
   AnimStateNode_4 = West() {is=title AnimStateNode}
   AnimStateNode_5 = JogLoop_3() {is=title AnimStateNode}
   AnimStateNode_7 = JogLoop() {is=title AnimStateNode}
   AnimStateNode_8 = JogLoop_1() {is=title AnimStateNode}
   AnimStateNode_9 = JogLoop_2() {is=title AnimStateNode}
   AnimStateTransitionNode_0 = AnimStateTransitionNode()
   AnimStateTransitionNode_1 = AnimStateTransitionNode()
   AnimStateTransitionNode_10 = AnimStateTransitionNode()
   AnimStateTransitionNode_11 = AnimStateTransitionNode()
   AnimStateTransitionNode_12 = AnimStateTransitionNode()
   AnimStateTransitionNode_2 = AnimStateTransitionNode()
   AnimStateTransitionNode_3 = AnimStateTransitionNode()
   AnimStateTransitionNode_4 = AnimStateTransitionNode() {bAutomaticRuleBasedOnSequencePlayerInState=true, crossfadeDuration=0.53}
   AnimStateTransitionNode_6 = AnimStateTransitionNode() {bAutomaticRuleBasedOnSequencePlayerInState=true, crossfadeDuration=0.53}
   AnimStateTransitionNode_7 = AnimStateTransitionNode() {bAutomaticRuleBasedOnSequencePlayerInState=true, crossfadeDuration=0.53}
   AnimStateTransitionNode_8 = AnimStateTransitionNode() {bAutomaticRuleBasedOnSequencePlayerInState=true, crossfadeDuration=0.53}
   AnimStateTransitionNode_9 = AnimStateTransitionNode()
  ]
  links [
   {
    dst=JogLoop
    rule=AnimGraph.AnimGraphNode_StateMachine_0.Locomotion.AnimStateNode_14.IntroRun.AnimGraphNode_StateMachine_0.New State Machine.AnimStateTransitionNode_4.Transition
    src=East
    via=AnimStateTransitionNode_4
   }
   {
    dst=State
    rule=AnimGraph.AnimGraphNode_StateMachine_0.Locomotion.AnimStateNode_14.IntroRun.AnimGraphNode_StateMachine_0.New State Machine.AnimStateTransitionNode_10.Transition
    src=JogLoop
    via=AnimStateTransitionNode_10
   }
   {
    dst=State
    rule=AnimGraph.AnimGraphNode_StateMachine_0.Locomotion.AnimStateNode_14.IntroRun.AnimGraphNode_StateMachine_0.New State Machine.AnimStateTransitionNode_9.Transition
    src=JogLoop_1
    via=AnimStateTransitionNode_9
   }
   {
    dst=State
    rule=AnimGraph.AnimGraphNode_StateMachine_0.Locomotion.AnimStateNode_14.IntroRun.AnimGraphNode_StateMachine_0.New State Machine.AnimStateTransitionNode_12.Transition
    src=JogLoop_2
    via=AnimStateTransitionNode_12
   }
   {
    dst=State
    rule=AnimGraph.AnimGraphNode_StateMachine_0.Locomotion.AnimStateNode_14.IntroRun.AnimGraphNode_StateMachine_0.New State Machine.AnimStateTransitionNode_11.Transition
    src=JogLoop_3
    via=AnimStateTransitionNode_11
   }
   {
    dst=JogLoop_1
    rule=AnimGraph.AnimGraphNode_StateMachine_0.Locomotion.AnimStateNode_14.IntroRun.AnimGraphNode_StateMachine_0.New State Machine.AnimStateTransitionNode_6.Transition
    src=North
    via=AnimStateTransitionNode_6
   }
   {
    dst=JogLoop_3
    rule=AnimGraph.AnimGraphNode_StateMachine_0.Locomotion.AnimStateNode_14.IntroRun.AnimGraphNode_StateMachine_0.New State Machine.AnimStateTransitionNode_7.Transition
    src=South
    via=AnimStateTransitionNode_7
   }
   {
    dst=East
    rule=AnimGraph.AnimGraphNode_StateMachine_0.Locomotion.AnimStateNode_14.IntroRun.AnimGraphNode_StateMachine_0.New State Machine.AnimStateTransitionNode_2.Transition
    src=State
    via=AnimStateTransitionNode_2
   }
   {
    dst=North
    rule=AnimGraph.AnimGraphNode_StateMachine_0.Locomotion.AnimStateNode_14.IntroRun.AnimGraphNode_StateMachine_0.New State Machine.AnimStateTransitionNode_0.Transition
    src=State
    via=AnimStateTransitionNode_0
   }
   {
    dst=South
    rule=AnimGraph.AnimGraphNode_StateMachine_0.Locomotion.AnimStateNode_14.IntroRun.AnimGraphNode_StateMachine_0.New State Machine.AnimStateTransitionNode_1.Transition
    src=State
    via=AnimStateTransitionNode_1
   }
   {
    dst=West
    rule=AnimGraph.AnimGraphNode_StateMachine_0.Locomotion.AnimStateNode_14.IntroRun.AnimGraphNode_StateMachine_0.New State Machine.AnimStateTransitionNode_3.Transition
    src=State
    via=AnimStateTransitionNode_3
   }
   {
    dst=JogLoop_2
    rule=AnimGraph.AnimGraphNode_StateMachine_0.Locomotion.AnimStateNode_14.IntroRun.AnimGraphNode_StateMachine_0.New State Machine.AnimStateTransitionNode_8.Transition
    src=West
    via=AnimStateTransitionNode_8
   }
  ]
 }
 State {
  path=AnimGraph.AnimGraphNode_StateMachine_0.Locomotion.AnimStateNode_14.IntroRun.AnimGraphNode_StateMachine_0.New State Machine.AnimStateNode_0.State
  nodes [Output Animation Pose(Result=Sequence Player 'Wayfare_Mask_TravelRun_Loop_Pack'() {sequence=/BRCosmetics/Accessories/FORT_Backpacks/Backpack_F_MED_Wayfare_Mask/Animations/Wayfare_Mask_TravelRun_Loop_Pack})]
 }
 North {
  path=AnimGraph.AnimGraphNode_StateMachine_0.Locomotion.AnimStateNode_14.IntroRun.AnimGraphNode_StateMachine_0.New State Machine.AnimStateNode_1.North
  nodes [
   n1 = Sequence Player 'Wayfare_Mask_TravelRun_Into_FWD'!(bLoopAnimation:bool=False) -> (Pose) {
    bLoopAnimation=false
    sequence=&Wayfare_Mask_TravelRun_Into_FWD
   }
   Output Animation Pose(Result=@n1)
  ]
 }
 South {
  path=AnimGraph.AnimGraphNode_StateMachine_0.Locomotion.AnimStateNode_14.IntroRun.AnimGraphNode_StateMachine_0.New State Machine.AnimStateNode_2.South
  nodes [
   n1 = Sequence Player 'Wayfare_Mask_TravelRun_Intro_Back'!(bLoopAnimation:bool=False) -> (Pose) {
    bLoopAnimation=false
    sequence=&Wayfare_Mask_TravelRun_Intro_Back
   }
   Output Animation Pose(Result=@n1)
  ]
 }
 East {
  path=AnimGraph.AnimGraphNode_StateMachine_0.Locomotion.AnimStateNode_14.IntroRun.AnimGraphNode_StateMachine_0.New State Machine.AnimStateNode_3.East
  nodes [
   n1 = Sequence Player 'Wayfare_Mask_TravelRun_Intro_Right'!(bLoopAnimation:bool=False) -> (Pose) {
    bLoopAnimation=false
    sequence=&Wayfare_Mask_TravelRun_Intro_Right
   }
   Output Animation Pose(Result=@n1)
  ]
 }
 West {
  path=AnimGraph.AnimGraphNode_StateMachine_0.Locomotion.AnimStateNode_14.IntroRun.AnimGraphNode_StateMachine_0.New State Machine.AnimStateNode_4.West
  nodes [
   n1 = Sequence Player 'Wayfare_Mask_TravelRun_Intro_Left'!(bLoopAnimation:bool=False) -> (Pose) {
    bLoopAnimation=false
    sequence=&Wayfare_Mask_TravelRun_Intro_Left
   }
   Output Animation Pose(Result=@n1)
  ]
 }
 JogLoop_3 {
  path=AnimGraph.AnimGraphNode_StateMachine_0.Locomotion.AnimStateNode_14.IntroRun.AnimGraphNode_StateMachine_0.New State Machine.AnimStateNode_5.JogLoop_3
  nodes [Output Animation Pose(Result=Sequence Player 'Wayfare_Mask_TravelRun_Loop_Pack'() {sequence=&Wayfare_Mask_TravelRun_Loop_Pack})]
 }
 JogLoop {
  path=AnimGraph.AnimGraphNode_StateMachine_0.Locomotion.AnimStateNode_14.IntroRun.AnimGraphNode_StateMachine_0.New State Machine.AnimStateNode_7.JogLoop
  nodes [Output Animation Pose(Result=Sequence Player 'Wayfare_Mask_TravelRun_Loop_Pack'() {sequence=&Wayfare_Mask_TravelRun_Loop_Pack})]
 }
 JogLoop_1 {
  path=AnimGraph.AnimGraphNode_StateMachine_0.Locomotion.AnimStateNode_14.IntroRun.AnimGraphNode_StateMachine_0.New State Machine.AnimStateNode_8.JogLoop_1
  nodes [Output Animation Pose(Result=Sequence Player 'Wayfare_Mask_TravelRun_Loop_Pack'() {sequence=&Wayfare_Mask_TravelRun_Loop_Pack})]
 }
 JogLoop_2 {
  path=AnimGraph.AnimGraphNode_StateMachine_0.Locomotion.AnimStateNode_14.IntroRun.AnimGraphNode_StateMachine_0.New State Machine.AnimStateNode_9.JogLoop_2
  nodes [Output Animation Pose(Result=Sequence Player 'Wayfare_Mask_TravelRun_Loop_Pack'() {sequence=&Wayfare_Mask_TravelRun_Loop_Pack})]
 }
 Transition {
  path=AnimGraph.AnimGraphNode_StateMachine_0.Locomotion.AnimStateNode_14.IntroRun.AnimGraphNode_StateMachine_0.New State Machine.AnimStateTransitionNode_0.Transition
  nodes [
   n1 = Get As Fortnite M Avg Player Anim Blueprint()
   Result(bCanEnterTransition="Equal (Enum)"(
     A:EFortCardinalDirection Enum=Get LocomotionCardinalDirection(self=@n1) -> (LocomotionCardinalDirection=North)
     B:EFortCardinalDirection Enum=North
    ))
  ]
 }
 Transition {
  path=AnimGraph.AnimGraphNode_StateMachine_0.Locomotion.AnimStateNode_14.IntroRun.AnimGraphNode_StateMachine_0.New State Machine.AnimStateTransitionNode_1.Transition
  nodes [
   n1 = Get As Fortnite M Avg Player Anim Blueprint()
   Result(bCanEnterTransition="Equal (Enum)"(A=Get LocomotionCardinalDirection(self=@n1) -> (LocomotionCardinalDirection=North), B=South))
  ]
 }
 Transition {
  path=AnimGraph.AnimGraphNode_StateMachine_0.Locomotion.AnimStateNode_14.IntroRun.AnimGraphNode_StateMachine_0.New State Machine.AnimStateTransitionNode_10.Transition
  nodes [
   n1 = Get LocomotionCardinalDirection(self=Get As Fortnite M Avg Player Anim Blueprint()) -> (LocomotionCardinalDirection=North)
   Result(bCanEnterTransition=NOT Boolean(A="Equal (Enum)"(A=@n1, B=East)))
  ]
 }
 Transition {
  path=AnimGraph.AnimGraphNode_StateMachine_0.Locomotion.AnimStateNode_14.IntroRun.AnimGraphNode_StateMachine_0.New State Machine.AnimStateTransitionNode_11.Transition
  nodes [
   n1 = Get LocomotionCardinalDirection(self=Get As Fortnite M Avg Player Anim Blueprint()) -> (LocomotionCardinalDirection=North)
   Result(bCanEnterTransition=NOT Boolean(A="Equal (Enum)"(A=@n1, B=South)))
  ]
 }
 Transition {
  path=AnimGraph.AnimGraphNode_StateMachine_0.Locomotion.AnimStateNode_14.IntroRun.AnimGraphNode_StateMachine_0.New State Machine.AnimStateTransitionNode_12.Transition
  nodes [
   n1 = Get LocomotionCardinalDirection(self=Get As Fortnite M Avg Player Anim Blueprint()) -> (LocomotionCardinalDirection=North)
   Result(bCanEnterTransition=NOT Boolean(A="Equal (Enum)"(A=@n1, B=West)))
  ]
 }
 Transition {
  path=AnimGraph.AnimGraphNode_StateMachine_0.Locomotion.AnimStateNode_14.IntroRun.AnimGraphNode_StateMachine_0.New State Machine.AnimStateTransitionNode_2.Transition
  nodes [
   n1 = Get As Fortnite M Avg Player Anim Blueprint()
   Result(bCanEnterTransition="Equal (Enum)"(A=Get LocomotionCardinalDirection(self=@n1) -> (LocomotionCardinalDirection=North), B=East))
  ]
 }
 Transition {
  path=AnimGraph.AnimGraphNode_StateMachine_0.Locomotion.AnimStateNode_14.IntroRun.AnimGraphNode_StateMachine_0.New State Machine.AnimStateTransitionNode_3.Transition
  nodes [
   n1 = Get As Fortnite M Avg Player Anim Blueprint()
   Result(bCanEnterTransition="Equal (Enum)"(A=Get LocomotionCardinalDirection(self=@n1) -> (LocomotionCardinalDirection=North), B=West))
  ]
 }
 Transition {
  path=AnimGraph.AnimGraphNode_StateMachine_0.Locomotion.AnimStateNode_14.IntroRun.AnimGraphNode_StateMachine_0.New State Machine.AnimStateTransitionNode_4.Transition
  nodes [Result()]
 }
 Transition {
  path=AnimGraph.AnimGraphNode_StateMachine_0.Locomotion.AnimStateNode_14.IntroRun.AnimGraphNode_StateMachine_0.New State Machine.AnimStateTransitionNode_6.Transition
  nodes [Result()]
 }
 Transition {
  path=AnimGraph.AnimGraphNode_StateMachine_0.Locomotion.AnimStateNode_14.IntroRun.AnimGraphNode_StateMachine_0.New State Machine.AnimStateTransitionNode_7.Transition
  nodes [Result()]
 }
 Transition {
  path=AnimGraph.AnimGraphNode_StateMachine_0.Locomotion.AnimStateNode_14.IntroRun.AnimGraphNode_StateMachine_0.New State Machine.AnimStateTransitionNode_8.Transition
  nodes [Result()]
 }
 Transition {
  path=AnimGraph.AnimGraphNode_StateMachine_0.Locomotion.AnimStateNode_14.IntroRun.AnimGraphNode_StateMachine_0.New State Machine.AnimStateTransitionNode_9.Transition
  nodes [
   n1 = Get LocomotionCardinalDirection(self=Get As Fortnite M Avg Player Anim Blueprint()) -> (LocomotionCardinalDirection=North)
   Result(bCanEnterTransition=NOT Boolean(A="Equal (Enum)"(A=@n1, B=North)))
  ]
 }
 Locomotion Placeholder To Save Transitions {
  path=AnimGraph.AnimGraphNode_StateMachine_0.Locomotion.AnimStateNode_15.Locomotion Placeholder To Save Transitions
  nodes [Output Animation Pose(Result=Sequence Player 'Wayfare_Mask_Idle'() {sequence=&Wayfare_Mask_Idle})]
 }
 JumpOutro {
  path=AnimGraph.AnimGraphNode_StateMachine_0.Locomotion.AnimStateNode_2.JumpOutro
  nodes [
   n1 = Sequence Player 'Wayfare_Mask_JumpOutro_PACK'!(bLoopAnimation:bool=False) -> (Pose) {bLoopAnimation=false, sequence=/BRCosmetics/Accessories/FORT_Backpacks/Backpack_F_MED_Wayfare_Mask/Animations/Wayfare_Mask_JumpOutro_PACK}
   Output Animation Pose(Result=@n1)
  ]
 }
 OutroRun {
  path=AnimGraph.AnimGraphNode_StateMachine_0.Locomotion.AnimStateNode_3.OutroRun
  nodes [
   n1 = Sequence Player 'Wayfare_Mask_TravelRun_Outro_Pack'() {
    bLoopAnimation=false
    sequence=/BRCosmetics/Accessories/FORT_Backpacks/Backpack_F_MED_Wayfare_Mask/Animations/Wayfare_Mask_TravelRun_Outro_Pack
   }
   Output Animation Pose(Result=@n1)
  ]
 }
 SkyDiving {
  path=AnimGraph.AnimGraphNode_StateMachine_0.Locomotion.AnimStateNode_4.SkyDiving
  nodes [Output Animation Pose(Result=Sequence Player 'Wayfare_Mask_JumpLoop_PACK'() {sequence=&Wayfare_Mask_JumpLoop_PACK})]
 }
 SkyDiveOutro {
  path=AnimGraph.AnimGraphNode_StateMachine_0.Locomotion.AnimStateNode_5.SkyDiveOutro
  nodes [Output Animation Pose(Result=Sequence Player 'Wayfare_Mask_JumpOutro_PACK'!() -> (Pose) {sequence=&Wayfare_Mask_JumpOutro_PACK})]
 }
 SkyDivingIntro {path=AnimGraph.AnimGraphNode_StateMachine_0.Locomotion.AnimStateNode_6.SkyDivingIntro, nodes [Output Animation Pose()]}
 Jump {
  path=AnimGraph.AnimGraphNode_StateMachine_0.Locomotion.AnimStateNode_7.Jump
  nodes [
   AnimGraphNode_StateMachine_0 = Jump State Machine() {is=title AnimGraphNode_StateMachine, bAllowConduitEntryStates=true}
   Output Animation Pose(Result=@AnimGraphNode_StateMachine_0)
  ]
 }
 Jump State Machine {
  entry=Conduit
  path=AnimGraph.AnimGraphNode_StateMachine_0.Locomotion.AnimStateNode_7.Jump.AnimGraphNode_StateMachine_0.Jump State Machine
  ids=id
  nodes [
   AnimStateConduitNode_0 = Conduit()
   AnimStateNode_0 = JumpIntro()
   AnimStateNode_1 = JumpLoop()
   AnimStateTransitionNode_16 = AnimStateTransitionNode() {bAutomaticRuleBasedOnSequencePlayerInState=true}
   AnimStateTransitionNode_18 = AnimStateTransitionNode()
   AnimStateTransitionNode_19 = AnimStateTransitionNode()
  ]
  links [
   {
    dst=JumpIntro
    rule=AnimGraph.AnimGraphNode_StateMachine_0.Locomotion.AnimStateNode_7.Jump.AnimGraphNode_StateMachine_0.Jump State Machine.AnimStateTransitionNode_18.Transition
    src=Conduit
    via=AnimStateTransitionNode_18
   }
   {
    dst=JumpLoop
    rule=AnimGraph.AnimGraphNode_StateMachine_0.Locomotion.AnimStateNode_7.Jump.AnimGraphNode_StateMachine_0.Jump State Machine.AnimStateTransitionNode_19.Transition
    src=Conduit
    via=AnimStateTransitionNode_19
   }
   {
    dst=JumpLoop
    rule=AnimGraph.AnimGraphNode_StateMachine_0.Locomotion.AnimStateNode_7.Jump.AnimGraphNode_StateMachine_0.Jump State Machine.AnimStateTransitionNode_16.Transition
    src=JumpIntro
    via=AnimStateTransitionNode_16
   }
  ]
 }
 Conduit {
  path=AnimGraph.AnimGraphNode_StateMachine_0.Locomotion.AnimStateNode_7.Jump.AnimGraphNode_StateMachine_0.Jump State Machine.AnimStateConduitNode_0.Conduit
  nodes [Result(bCanEnterTransition=true)]
 }
 JumpIntro {
  path=AnimGraph.AnimGraphNode_StateMachine_0.Locomotion.AnimStateNode_7.Jump.AnimGraphNode_StateMachine_0.Jump State Machine.AnimStateNode_0.JumpIntro
  nodes [
   n1 = Sequence Player 'Wayfare_Mask_JumpInto_PACK'(bLoopAnimation=False) {bLoopAnimation=false, sequence=&Wayfare_Mask_JumpInto_PACK}
   Output Animation Pose(Result=@n1)
  ]
 }
 JumpLoop {
  path=AnimGraph.AnimGraphNode_StateMachine_0.Locomotion.AnimStateNode_7.Jump.AnimGraphNode_StateMachine_0.Jump State Machine.AnimStateNode_1.JumpLoop
  nodes [Output Animation Pose(Result=Sequence Player 'Wayfare_Mask_JumpLoop_PACK'() {sequence=&Wayfare_Mask_JumpLoop_PACK})]
 }
 Transition {
  path=AnimGraph.AnimGraphNode_StateMachine_0.Locomotion.AnimStateNode_7.Jump.AnimGraphNode_StateMachine_0.Jump State Machine.AnimStateTransitionNode_16.Transition
  nodes [
   Result()
   Get As Fortnite M Avg Player Anim Blueprint()
  ]
 }
 Transition {
  path=AnimGraph.AnimGraphNode_StateMachine_0.Locomotion.AnimStateNode_7.Jump.AnimGraphNode_StateMachine_0.Jump State Machine.AnimStateTransitionNode_18.Transition
  nodes [Result(bCanEnterTransition=Get bIsJumping!(self:Fort Anim Instance Object Reference=Get As Fortnite M Avg Player Anim Blueprint()) -> (bIsJumping))]
 }
 Transition {
  path=AnimGraph.AnimGraphNode_StateMachine_0.Locomotion.AnimStateNode_7.Jump.AnimGraphNode_StateMachine_0.Jump State Machine.AnimStateTransitionNode_19.Transition
  nodes [Result(bCanEnterTransition=Get bIsFalling!(self:Fort Anim Instance Object Reference=Get As Fortnite M Avg Player Anim Blueprint()) -> (bIsFalling))]
 }
 RunLoop {
  path=AnimGraph.AnimGraphNode_StateMachine_0.Locomotion.AnimStateNode_8.RunLoop
  nodes [Output Animation Pose(Result=Sequence Player 'Wayfare_Mask_TravelRun_Loop_Pack'() {sequence=&Wayfare_Mask_TravelRun_Loop_Pack})]
 }
 Locomotion {
  path=AnimGraph.AnimGraphNode_StateMachine_0.Locomotion.AnimStateNode_9.Locomotion
  nodes [
   AnimGraphNode_StateMachine_0 = Locomotion State Machine() {is=title AnimGraphNode_StateMachine}
   Output Animation Pose(Result=@AnimGraphNode_StateMachine_0)
  ]
 }
 Locomotion State Machine {
  entry=Idle
  path=AnimGraph.AnimGraphNode_StateMachine_0.Locomotion.AnimStateNode_9.Locomotion.AnimGraphNode_StateMachine_0.Locomotion State Machine
  ids=id
  nodes [
   AnimStateNode_0 = Idle()
   AnimStateNode_1 = IntroRun() {stateEntered {guid=#24, notifyName=Entered_IntroRun}}
   AnimStateNode_2 = IdleBreak() {is=title AnimStateNode}
   AnimStateNode_3 = OutroRun() {stateEntered {guid=#25, notifyName=Entered_OutroRun}}
   AnimStateNode_7 = JogLoop() {stateEntered {guid=#26, notifyName=Entered_JogLoop}}
   AnimStateTransitionNode_0 = AnimStateTransitionNode()
   AnimStateTransitionNode_1 = AnimStateTransitionNode() {bAutomaticRuleBasedOnSequencePlayerInState=true, crossfadeDuration=0.35}
   AnimStateTransitionNode_2 = AnimStateTransitionNode() {bAutomaticRuleBasedOnSequencePlayerInState=true, crossfadeDuration=0.8}
   AnimStateTransitionNode_27 = AnimStateTransitionNode() {bAutomaticRuleBasedOnSequencePlayerInState=true, crossfadeDuration=0.8}
   AnimStateTransitionNode_28 = AnimStateTransitionNode()
   AnimStateTransitionNode_3 = AnimStateTransitionNode() {bAutomaticRuleBasedOnSequencePlayerInState=true, crossfadeDuration=0.73}
   AnimStateTransitionNode_33 = AnimStateTransitionNode()
   AnimStateTransitionNode_4 = AnimStateTransitionNode()
  ]
  links [
   {
    dst=IdleBreak
    rule=AnimGraph.AnimGraphNode_StateMachine_0.Locomotion.AnimStateNode_9.Locomotion.AnimGraphNode_StateMachine_0.Locomotion State Machine.AnimStateTransitionNode_1.AnimationTransitionGraph_0
    src=Idle
    via=AnimStateTransitionNode_1
   }
   {
    dst=IntroRun
    rule=AnimGraph.AnimGraphNode_StateMachine_0.Locomotion.AnimStateNode_9.Locomotion.AnimGraphNode_StateMachine_0.Locomotion State Machine.AnimStateTransitionNode_0.AnimationTransitionGraph_0
    src=Idle
    via=AnimStateTransitionNode_0
   }
   {
    dst=Idle
    rule=AnimGraph.AnimGraphNode_StateMachine_0.Locomotion.AnimStateNode_9.Locomotion.AnimGraphNode_StateMachine_0.Locomotion State Machine.AnimStateTransitionNode_2.AnimationTransitionGraph_0
    src=IdleBreak
    via=AnimStateTransitionNode_2
   }
   {
    dst=JogLoop
    rule=AnimGraph.AnimGraphNode_StateMachine_0.Locomotion.AnimStateNode_9.Locomotion.AnimGraphNode_StateMachine_0.Locomotion State Machine.AnimStateTransitionNode_27.Transition
    src=IntroRun
    via=AnimStateTransitionNode_27
   }
   {
    dst=JogLoop
    rule=AnimGraph.AnimGraphNode_StateMachine_0.Locomotion.AnimStateNode_9.Locomotion.AnimGraphNode_StateMachine_0.Locomotion State Machine.AnimStateTransitionNode_4.Transition
    src=IntroRun
    via=AnimStateTransitionNode_4
   }
   {
    dst=IntroRun
    rule=AnimGraph.AnimGraphNode_StateMachine_0.Locomotion.AnimStateNode_9.Locomotion.AnimGraphNode_StateMachine_0.Locomotion State Machine.AnimStateTransitionNode_33.Transition
    src=JogLoop
    via=AnimStateTransitionNode_33
   }
   {
    dst=OutroRun
    rule=AnimGraph.AnimGraphNode_StateMachine_0.Locomotion.AnimStateNode_9.Locomotion.AnimGraphNode_StateMachine_0.Locomotion State Machine.AnimStateTransitionNode_28.Transition
    src=JogLoop
    via=AnimStateTransitionNode_28
   }
   {
    dst=Idle
    rule=AnimGraph.AnimGraphNode_StateMachine_0.Locomotion.AnimStateNode_9.Locomotion.AnimGraphNode_StateMachine_0.Locomotion State Machine.AnimStateTransitionNode_3.AnimationTransitionGraph_0
    src=OutroRun
    via=AnimStateTransitionNode_3
   }
  ]
 }
 Idle {
  path=AnimGraph.AnimGraphNode_StateMachine_0.Locomotion.AnimStateNode_9.Locomotion.AnimGraphNode_StateMachine_0.Locomotion State Machine.AnimStateNode_0.Idle
  nodes [Output Animation Pose(Result=Sequence Player 'Wayfare_Mask_Idle'() {sequence=&Wayfare_Mask_Idle})]
 }
 IntroRun {
  path=AnimGraph.AnimGraphNode_StateMachine_0.Locomotion.AnimStateNode_9.Locomotion.AnimGraphNode_StateMachine_0.Locomotion State Machine.AnimStateNode_1.IntroRun
  nodes [
   AnimGraphNode_StateMachine_0 = New State Machine() {bAllowConduitEntryStates=true}
   Output Animation Pose(Result=@AnimGraphNode_StateMachine_0)
  ]
 }
 New State Machine {
  entry=Conduit
  path=AnimGraph.AnimGraphNode_StateMachine_0.Locomotion.AnimStateNode_9.Locomotion.AnimGraphNode_StateMachine_0.Locomotion State Machine.AnimStateNode_1.IntroRun.AnimGraphNode_StateMachine_0.New State Machine
  ids=id
  nodes [
   AnimStateConduitNode_1 = Conduit()
   AnimStateNode_1 = North()
   AnimStateNode_2 = South()
   AnimStateNode_3 = East()
   AnimStateNode_4 = West()
   AnimStateTransitionNode_29 = AnimStateTransitionNode()
   AnimStateTransitionNode_30 = AnimStateTransitionNode()
   AnimStateTransitionNode_31 = AnimStateTransitionNode()
   AnimStateTransitionNode_32 = AnimStateTransitionNode()
  ]
  links [
   {
    dst=East
    rule=AnimGraph.AnimGraphNode_StateMachine_0.Locomotion.AnimStateNode_9.Locomotion.AnimGraphNode_StateMachine_0.Locomotion State Machine.AnimStateNode_1.IntroRun.AnimGraphNode_StateMachine_0.New State Machine.AnimStateTransitionNode_30.Transition
    src=Conduit
    via=AnimStateTransitionNode_30
   }
   {
    dst=North
    rule=AnimGraph.AnimGraphNode_StateMachine_0.Locomotion.AnimStateNode_9.Locomotion.AnimGraphNode_StateMachine_0.Locomotion State Machine.AnimStateNode_1.IntroRun.AnimGraphNode_StateMachine_0.New State Machine.AnimStateTransitionNode_29.Transition
    src=Conduit
    via=AnimStateTransitionNode_29
   }
   {
    dst=South
    rule=AnimGraph.AnimGraphNode_StateMachine_0.Locomotion.AnimStateNode_9.Locomotion.AnimGraphNode_StateMachine_0.Locomotion State Machine.AnimStateNode_1.IntroRun.AnimGraphNode_StateMachine_0.New State Machine.AnimStateTransitionNode_31.Transition
    src=Conduit
    via=AnimStateTransitionNode_31
   }
   {
    dst=West
    rule=AnimGraph.AnimGraphNode_StateMachine_0.Locomotion.AnimStateNode_9.Locomotion.AnimGraphNode_StateMachine_0.Locomotion State Machine.AnimStateNode_1.IntroRun.AnimGraphNode_StateMachine_0.New State Machine.AnimStateTransitionNode_32.Transition
    src=Conduit
    via=AnimStateTransitionNode_32
   }
  ]
 }
 Conduit {
  path=AnimGraph.AnimGraphNode_StateMachine_0.Locomotion.AnimStateNode_9.Locomotion.AnimGraphNode_StateMachine_0.Locomotion State Machine.AnimStateNode_1.IntroRun.AnimGraphNode_StateMachine_0.New State Machine.AnimStateConduitNode_1.Conduit
  nodes [Result(bCanEnterTransition=true)]
 }
 North {
  path=AnimGraph.AnimGraphNode_StateMachine_0.Locomotion.AnimStateNode_9.Locomotion.AnimGraphNode_StateMachine_0.Locomotion State Machine.AnimStateNode_1.IntroRun.AnimGraphNode_StateMachine_0.New State Machine.AnimStateNode_1.North
  nodes [
   n1 = Sequence Player 'Wayfare_Mask_TravelRun_Into_FWD'(bLoopAnimation=False) {
    bLoopAnimation=false
    sequence=&Wayfare_Mask_TravelRun_Into_FWD
   }
   Output Animation Pose(Result=@n1)
  ]
 }
 South {
  path=AnimGraph.AnimGraphNode_StateMachine_0.Locomotion.AnimStateNode_9.Locomotion.AnimGraphNode_StateMachine_0.Locomotion State Machine.AnimStateNode_1.IntroRun.AnimGraphNode_StateMachine_0.New State Machine.AnimStateNode_2.South
  nodes [
   n1 = Sequence Player 'Wayfare_Mask_TravelRun_Intro_Back'(bLoopAnimation=False) {
    bLoopAnimation=false
    sequence=&Wayfare_Mask_TravelRun_Intro_Back
   }
   Output Animation Pose(Result=@n1)
  ]
 }
 East {
  path=AnimGraph.AnimGraphNode_StateMachine_0.Locomotion.AnimStateNode_9.Locomotion.AnimGraphNode_StateMachine_0.Locomotion State Machine.AnimStateNode_1.IntroRun.AnimGraphNode_StateMachine_0.New State Machine.AnimStateNode_3.East
  nodes [
   n1 = Sequence Player 'Wayfare_Mask_TravelRun_Intro_Right'(bLoopAnimation=False) {
    bLoopAnimation=false
    sequence=&Wayfare_Mask_TravelRun_Intro_Right
   }
   Output Animation Pose(Result=@n1)
  ]
 }
 West {
  path=AnimGraph.AnimGraphNode_StateMachine_0.Locomotion.AnimStateNode_9.Locomotion.AnimGraphNode_StateMachine_0.Locomotion State Machine.AnimStateNode_1.IntroRun.AnimGraphNode_StateMachine_0.New State Machine.AnimStateNode_4.West
  nodes [
   n1 = Sequence Player 'Wayfare_Mask_TravelRun_Intro_Left'(bLoopAnimation=False) {
    bLoopAnimation=false
    sequence=&Wayfare_Mask_TravelRun_Intro_Left
   }
   Output Animation Pose(Result=@n1)
  ]
 }
 Transition {
  path=AnimGraph.AnimGraphNode_StateMachine_0.Locomotion.AnimStateNode_9.Locomotion.AnimGraphNode_StateMachine_0.Locomotion State Machine.AnimStateNode_1.IntroRun.AnimGraphNode_StateMachine_0.New State Machine.AnimStateTransitionNode_29.Transition
  nodes [
   n1 = Get As Fortnite M Avg Player Anim Blueprint()
   Result(bCanEnterTransition="Equal (Enum)"(A=Get LocomotionCardinalDirection(self=@n1) -> (LocomotionCardinalDirection=North), B=North))
  ]
 }
 Transition {
  path=AnimGraph.AnimGraphNode_StateMachine_0.Locomotion.AnimStateNode_9.Locomotion.AnimGraphNode_StateMachine_0.Locomotion State Machine.AnimStateNode_1.IntroRun.AnimGraphNode_StateMachine_0.New State Machine.AnimStateTransitionNode_30.Transition
  nodes [
   n1 = Get As Fortnite M Avg Player Anim Blueprint()
   Result(bCanEnterTransition="Equal (Enum)"(A=Get LocomotionCardinalDirection(self=@n1) -> (LocomotionCardinalDirection=North), B=East))
  ]
 }
 Transition {
  path=AnimGraph.AnimGraphNode_StateMachine_0.Locomotion.AnimStateNode_9.Locomotion.AnimGraphNode_StateMachine_0.Locomotion State Machine.AnimStateNode_1.IntroRun.AnimGraphNode_StateMachine_0.New State Machine.AnimStateTransitionNode_31.Transition
  nodes [
   n1 = Get As Fortnite M Avg Player Anim Blueprint()
   Result(bCanEnterTransition="Equal (Enum)"(A=Get LocomotionCardinalDirection(self=@n1) -> (LocomotionCardinalDirection=North), B=South))
  ]
 }
 Transition {
  path=AnimGraph.AnimGraphNode_StateMachine_0.Locomotion.AnimStateNode_9.Locomotion.AnimGraphNode_StateMachine_0.Locomotion State Machine.AnimStateNode_1.IntroRun.AnimGraphNode_StateMachine_0.New State Machine.AnimStateTransitionNode_32.Transition
  nodes [
   n1 = Get As Fortnite M Avg Player Anim Blueprint()
   Result(bCanEnterTransition="Equal (Enum)"(A=Get LocomotionCardinalDirection(self=@n1) -> (LocomotionCardinalDirection=North), B=West))
  ]
 }
 IdleBreak {
  path=AnimGraph.AnimGraphNode_StateMachine_0.Locomotion.AnimStateNode_9.Locomotion.AnimGraphNode_StateMachine_0.Locomotion State Machine.AnimStateNode_2.IdleBreak
  nodes [Output Animation Pose(Result=Sequence Player 'Wayfare_Mask_TravelRun_IdleBreak'() {sequence=/BRCosmetics/Accessories/FORT_Backpacks/Backpack_F_MED_Wayfare_Mask/Animations/Wayfare_Mask_TravelRun_IdleBreak})]
 }
 OutroRun {
  path=AnimGraph.AnimGraphNode_StateMachine_0.Locomotion.AnimStateNode_9.Locomotion.AnimGraphNode_StateMachine_0.Locomotion State Machine.AnimStateNode_3.OutroRun
  nodes [
   n1 = Sequence Player 'Wayfare_Mask_TravelRun_Outro_Pack'() {
    bLoopAnimation=false
    sequence=&Wayfare_Mask_TravelRun_Outro_Pack
   }
   Output Animation Pose(Result=@n1)
  ]
 }
 JogLoop {
  path=AnimGraph.AnimGraphNode_StateMachine_0.Locomotion.AnimStateNode_9.Locomotion.AnimGraphNode_StateMachine_0.Locomotion State Machine.AnimStateNode_7.JogLoop
  nodes [Output Animation Pose(Result=Sequence Player 'Wayfare_Mask_TravelRun_Loop_Pack'() {sequence=&Wayfare_Mask_TravelRun_Loop_Pack})]
 }
 AnimationTransitionGraph_0 {
  path=AnimGraph.AnimGraphNode_StateMachine_0.Locomotion.AnimStateNode_9.Locomotion.AnimGraphNode_StateMachine_0.Locomotion State Machine.AnimStateTransitionNode_0.AnimationTransitionGraph_0
  nodes [Result(bCanEnterTransition=Get bIsMoving2D!(self:Fort Player Anim Instance Object Reference=Get As Fortnite M Avg Player Anim Blueprint()) -> (bIsMoving2D))]
 }
 AnimationTransitionGraph_0 {
  path=AnimGraph.AnimGraphNode_StateMachine_0.Locomotion.AnimStateNode_9.Locomotion.AnimGraphNode_StateMachine_0.Locomotion State Machine.AnimStateTransitionNode_1.AnimationTransitionGraph_0
  nodes [Result(bCanEnterTransition=true)]
 }
 AnimationTransitionGraph_0 {
  path=AnimGraph.AnimGraphNode_StateMachine_0.Locomotion.AnimStateNode_9.Locomotion.AnimGraphNode_StateMachine_0.Locomotion State Machine.AnimStateTransitionNode_2.AnimationTransitionGraph_0
  nodes [Result(bCanEnterTransition=true)]
 }
 Transition {
  path=AnimGraph.AnimGraphNode_StateMachine_0.Locomotion.AnimStateNode_9.Locomotion.AnimGraphNode_StateMachine_0.Locomotion State Machine.AnimStateTransitionNode_27.Transition
  nodes [Result()]
 }
 Transition {
  path=AnimGraph.AnimGraphNode_StateMachine_0.Locomotion.AnimStateNode_9.Locomotion.AnimGraphNode_StateMachine_0.Locomotion State Machine.AnimStateTransitionNode_28.Transition
  nodes [
   n1 = Get As Fortnite M Avg Player Anim Blueprint()
   Result(bCanEnterTransition=NOT Boolean(A=Get bIsMoving2D(self=@n1)))
  ]
 }
 AnimationTransitionGraph_0 {
  path=AnimGraph.AnimGraphNode_StateMachine_0.Locomotion.AnimStateNode_9.Locomotion.AnimGraphNode_StateMachine_0.Locomotion State Machine.AnimStateTransitionNode_3.AnimationTransitionGraph_0
  nodes [Result(bCanEnterTransition=true)]
 }
 Transition {
  path=AnimGraph.AnimGraphNode_StateMachine_0.Locomotion.AnimStateNode_9.Locomotion.AnimGraphNode_StateMachine_0.Locomotion State Machine.AnimStateTransitionNode_33.Transition
  nodes [Result(bCanEnterTransition=Get AllowDirectionChange() -> (AllowDirectionChange:bool))]
 }
 Transition {
  path=AnimGraph.AnimGraphNode_StateMachine_0.Locomotion.AnimStateNode_9.Locomotion.AnimGraphNode_StateMachine_0.Locomotion State Machine.AnimStateTransitionNode_4.Transition
  nodes [Result(bCanEnterTransition=Get JustLandedFromJump() -> (JustLandedFromJump:bool))]
 }
 AnimationTransitionGraph_0 {
  path=AnimGraph.AnimGraphNode_StateMachine_0.Locomotion.AnimStateTransitionNode_0.AnimationTransitionGraph_0
  nodes [Result(bCanEnterTransition=Get bIsMoving2D(self=Get As Fortnite M Avg Player Anim Blueprint()))]
 }
 Transition {
  path=AnimGraph.AnimGraphNode_StateMachine_0.Locomotion.AnimStateTransitionNode_10.Transition
  nodes [Result(bCanEnterTransition=Get bIsMoving2D!() -> (bIsMoving2D))]
 }
 Transition {path=AnimGraph.AnimGraphNode_StateMachine_0.Locomotion.AnimStateTransitionNode_11.Transition, nodes [Result(bCanEnterTransition=true)]}
 Transition {path=AnimGraph.AnimGraphNode_StateMachine_0.Locomotion.AnimStateTransitionNode_13.Transition, nodes [Result()]}
 AnimationTransitionGraph_1 {
  path=AnimGraph.AnimGraphNode_StateMachine_0.Locomotion.AnimStateTransitionNode_14.AnimationTransitionGraph_1
  nodes [
   n1 = Get As Fortnite M Avg Player Anim Blueprint()
   Result(bCanEnterTransition=NOT Boolean(A=Get bIsMoving2D!(self:Fort Player Anim Instance Object Reference=@n1) -> (bIsMoving2D)))
  ]
 }
 Transition {
  path=AnimGraph.AnimGraphNode_StateMachine_0.Locomotion.AnimStateTransitionNode_16.Transition
  nodes [Result(bCanEnterTransition=Get bIsJumping(self=Get As Fortnite M Avg Player Anim Blueprint()))]
 }
 Transition {
  path=AnimGraph.AnimGraphNode_StateMachine_0.Locomotion.AnimStateTransitionNode_17.Transition
  nodes [Result(bCanEnterTransition=Get bIsFalling(self=Get As Fortnite M Avg Player Anim Blueprint()))]
 }
 Transition {
  path=AnimGraph.AnimGraphNode_StateMachine_0.Locomotion.AnimStateTransitionNode_18.Transition
  nodes [
   n1 = Get As Fortnite M Avg Player Anim Blueprint()
   Result(bCanEnterTransition=OR Boolean(A=Get bIsOnGround!(self:Fort Player Anim Instance Object Reference=@n1) -> (bIsOnGround:bool), B=Get bIsSurfaceSwimming()))
  ]
 }
 Transition {
  path=AnimGraph.AnimGraphNode_StateMachine_0.Locomotion.AnimStateTransitionNode_19.Transition
  nodes [
   n1 = Get bIsOnGround(self=Get As Fortnite M Avg Player Anim Blueprint())
   n2 = Get bIsSurfaceSwimming()
   Result(bCanEnterTransition=AND Boolean(A=NOT Boolean(A=@n1), B=NOT Boolean(A=@n2)))
  ]
 }
 AnimationTransitionGraph_1 {
  path=AnimGraph.AnimGraphNode_StateMachine_0.Locomotion.AnimStateTransitionNode_2.AnimationTransitionGraph_1
  nodes [Result(bCanEnterTransition=NOT Boolean(A=Get bIsPlayingMeleeAnim()))]
 }
 Transition {
  path=AnimGraph.AnimGraphNode_StateMachine_0.Locomotion.AnimStateTransitionNode_20.Transition
  nodes [
   n1 = Get As Fortnite M Avg Player Anim Blueprint()
   Result(bCanEnterTransition=AND Boolean(A=Get bIsOnGround(self=@n1), B=Get bIsMoving2D(self=@n1)))
  ]
 }
 Transition {
  path=AnimGraph.AnimGraphNode_StateMachine_0.Locomotion.AnimStateTransitionNode_21.Transition
  nodes [
   n1 = Get As Fortnite M Avg Player Anim Blueprint()
   Result(bCanEnterTransition=OR Boolean(A=Get bIsOnGround(self=@n1), B=Get bIsSurfaceSwimming()))
  ]
 }
 Transition {path=AnimGraph.AnimGraphNode_StateMachine_0.Locomotion.AnimStateTransitionNode_25.Transition, nodes [Result()]}
 AnimationTransitionGraph_0 {
  path=AnimGraph.AnimGraphNode_StateMachine_0.Locomotion.AnimStateTransitionNode_3.AnimationTransitionGraph_0
  nodes [Result(bCanEnterTransition=true)]
 }
 AnimationTransitionGraph_1 {
  path=AnimGraph.AnimGraphNode_StateMachine_0.Locomotion.AnimStateTransitionNode_4.AnimationTransitionGraph_1
  nodes [Result(bCanEnterTransition=true)]
 }
 AnimationTransitionGraph_0 {
  path=AnimGraph.AnimGraphNode_StateMachine_0.Locomotion.AnimStateTransitionNode_5.AnimationTransitionGraph_0
  nodes [Result(bCanEnterTransition=Get bIsPlayingMeleeAnim())]
 }
 Transition {
  path=AnimGraph.AnimGraphNode_StateMachine_0.Locomotion.AnimStateTransitionNode_6.Transition
  nodes [
   n1 = Get As Fortnite M Avg Player Anim Blueprint()
   Result(bCanEnterTransition=OR Boolean(A=Get bIsOnGround(self=@n1), B=Get bIsSurfaceSwimming()))
  ]
 }
 Transition {
  path=AnimGraph.AnimGraphNode_StateMachine_0.Locomotion.AnimStateTransitionNode_7.Transition
  nodes [Result(bCanEnterTransition=NOT Boolean(A=Get bIsMoving2D!() -> (bIsMoving2D)))]
 }
 Transition {
  path=AnimGraph.AnimGraphNode_StateMachine_0.Locomotion.AnimStateTransitionNode_8.Transition
  nodes [Result(bCanEnterTransition=NOT Boolean(A=Get bIsSkydiveDiveMode()))]
 }
 Transition {path=AnimGraph.AnimGraphNode_StateMachine_0.Locomotion.AnimStateTransitionNode_9.Transition, nodes [Result()]}
 Gliding {
  entry=Locomotion
  path=AnimGraph.AnimGraphNode_StateMachine_1.Gliding
  ids=id
  nodes [
   AnimStateNode_0 = Locomotion()
   AnimStateNode_1 = Gliding() {is=title AnimStateNode}
   AnimStateNode_2 = Skydiving() {is=title AnimStateNode}
   AnimStateNode_3 = SkydivingOutro() {is=title AnimStateNode}
   AnimStateNode_4 = SkydiveInto() {is=title AnimStateNode}
   AnimStateTransitionNode_1 = AnimStateTransitionNode() {crossfadeDuration=0.3}
   AnimStateTransitionNode_4 = AnimStateTransitionNode() {crossfadeDuration=0.3}
   AnimStateTransitionNode_5 = AnimStateTransitionNode() {crossfadeDuration=0.8}
   AnimStateTransitionNode_6 = AnimStateTransitionNode() {bAutomaticRuleBasedOnSequencePlayerInState=true, crossfadeDuration=0.8}
   AnimStateTransitionNode_7 = AnimStateTransitionNode() {crossfadeDuration=0.8}
   AnimStateTransitionNode_8 = AnimStateTransitionNode() {bAutomaticRuleBasedOnSequencePlayerInState=true}
  ]
  links [
   {
    dst=Locomotion
    rule=AnimGraph.AnimGraphNode_StateMachine_1.Gliding.AnimStateTransitionNode_4.Transition
    src=Gliding
    via=AnimStateTransitionNode_4
   }
   {
    dst=SkydiveInto
    rule=AnimGraph.AnimGraphNode_StateMachine_1.Gliding.AnimStateTransitionNode_7.Transition
    src=Gliding
    via=AnimStateTransitionNode_7
   }
   {
    dst=Gliding
    rule=AnimGraph.AnimGraphNode_StateMachine_1.Gliding.AnimStateTransitionNode_1.Transition
    src=Locomotion
    via=AnimStateTransitionNode_1
   }
   {
    dst=Skydiving
    rule=AnimGraph.AnimGraphNode_StateMachine_1.Gliding.AnimStateTransitionNode_8.Transition
    src=SkydiveInto
    via=AnimStateTransitionNode_8
   }
   {
    dst=SkydivingOutro
    rule=AnimGraph.AnimGraphNode_StateMachine_1.Gliding.AnimStateTransitionNode_5.Transition
    src=Skydiving
    via=AnimStateTransitionNode_5
   }
   {
    dst=Gliding
    rule=AnimGraph.AnimGraphNode_StateMachine_1.Gliding.AnimStateTransitionNode_6.Transition
    src=SkydivingOutro
    via=AnimStateTransitionNode_6
   }
  ]
 }
 Locomotion {
  path=AnimGraph.AnimGraphNode_StateMachine_1.Gliding.AnimStateNode_0.Locomotion
  nodes [
   AnimGraphNode_StateMachine_3 = FrontEnd() {is=title AnimGraphNode_StateMachine}
   Output Animation Pose(Result=Blend Poses by bool(BlendPose_0=@AnimGraphNode_StateMachine_3, BlendPose_1=Use cached pose 'FullLocoCache'(), bActiveValue=Get bIsInFrontEnd()) {blendType=Linear})
  ]
 }
 FrontEnd {
  entry=Idle
  path=AnimGraph.AnimGraphNode_StateMachine_1.Gliding.AnimStateNode_0.Locomotion.AnimGraphNode_StateMachine_3.FrontEnd
  ids=id
  nodes [
   AnimStateNode_0 = Idle()
   AnimStateNode_2 = IdleBreak()
   AnimStateTransitionNode_1 = AnimStateTransitionNode() {bAutomaticRuleBasedOnSequencePlayerInState=true, crossfadeDuration=0.35}
   AnimStateTransitionNode_2 = AnimStateTransitionNode() {bAutomaticRuleBasedOnSequencePlayerInState=true, crossfadeDuration=0.8}
  ]
  links [
   {
    dst=IdleBreak
    rule=AnimGraph.AnimGraphNode_StateMachine_1.Gliding.AnimStateNode_0.Locomotion.AnimGraphNode_StateMachine_3.FrontEnd.AnimStateTransitionNode_1.AnimationTransitionGraph_0
    src=Idle
    via=AnimStateTransitionNode_1
   }
   {
    dst=Idle
    rule=AnimGraph.AnimGraphNode_StateMachine_1.Gliding.AnimStateNode_0.Locomotion.AnimGraphNode_StateMachine_3.FrontEnd.AnimStateTransitionNode_2.AnimationTransitionGraph_0
    src=IdleBreak
    via=AnimStateTransitionNode_2
   }
  ]
 }
 Idle {
  path=AnimGraph.AnimGraphNode_StateMachine_1.Gliding.AnimStateNode_0.Locomotion.AnimGraphNode_StateMachine_3.FrontEnd.AnimStateNode_0.Idle
  nodes [Output Animation Pose(Result=Sequence Player 'Wayfare_Mask_Idle'() {sequence=&Wayfare_Mask_Idle})]
 }
 IdleBreak {
  path=AnimGraph.AnimGraphNode_StateMachine_1.Gliding.AnimStateNode_0.Locomotion.AnimGraphNode_StateMachine_3.FrontEnd.AnimStateNode_2.IdleBreak
  nodes [Output Animation Pose(Result=Sequence Player 'Wayfare_Mask_TravelRun_IdleBreak'() {sequence=&Wayfare_Mask_TravelRun_IdleBreak})]
 }
 AnimationTransitionGraph_0 {
  path=AnimGraph.AnimGraphNode_StateMachine_1.Gliding.AnimStateNode_0.Locomotion.AnimGraphNode_StateMachine_3.FrontEnd.AnimStateTransitionNode_1.AnimationTransitionGraph_0
  nodes [Result(bCanEnterTransition=true)]
 }
 AnimationTransitionGraph_0 {
  path=AnimGraph.AnimGraphNode_StateMachine_1.Gliding.AnimStateNode_0.Locomotion.AnimGraphNode_StateMachine_3.FrontEnd.AnimStateTransitionNode_2.AnimationTransitionGraph_0
  nodes [Result(bCanEnterTransition=true)]
 }
 Gliding {
  path=AnimGraph.AnimGraphNode_StateMachine_1.Gliding.AnimStateNode_1.Gliding
  nodes [Output Animation Pose(Result=Sequence Player 'Wayfare_Mask_TravelRun_Loop_Pack'() {sequence=&Wayfare_Mask_TravelRun_Loop_Pack})]
 }
 Skydiving {
  path=AnimGraph.AnimGraphNode_StateMachine_1.Gliding.AnimStateNode_2.Skydiving
  nodes [Output Animation Pose(Result=Sequence Player 'Wayfare_Mask_JumpLoop_PACK'() {sequence=&Wayfare_Mask_JumpLoop_PACK})]
 }
 SkydivingOutro {
  path=AnimGraph.AnimGraphNode_StateMachine_1.Gliding.AnimStateNode_3.SkydivingOutro
  nodes [Output Animation Pose(Result=Sequence Player 'Wayfare_Mask_JumpOutro_PACK'() {sequence=&Wayfare_Mask_JumpOutro_PACK})]
 }
 SkydiveInto {
  path=AnimGraph.AnimGraphNode_StateMachine_1.Gliding.AnimStateNode_4.SkydiveInto
  nodes [Output Animation Pose(Result=Sequence Player 'Wayfare_Mask_JumpInto_PACK'!() -> (Pose) {sequence=&Wayfare_Mask_JumpInto_PACK})]
 }
 Transition {
  path=AnimGraph.AnimGraphNode_StateMachine_1.Gliding.AnimStateTransitionNode_1.Transition
  nodes [Result(bCanEnterTransition=Get bIsSkyDiving())]
 }
 Transition {
  path=AnimGraph.AnimGraphNode_StateMachine_1.Gliding.AnimStateTransitionNode_4.Transition
  nodes [Result(bCanEnterTransition=NOT Boolean(A=Get bIsSkyDiving()))]
 }
 Transition {
  path=AnimGraph.AnimGraphNode_StateMachine_1.Gliding.AnimStateTransitionNode_5.Transition
  nodes [Result(bCanEnterTransition=NOT Boolean(A=Get bIsSkydivingDiving() -> (bIsSkydivingDiving:bool)))]
 }
 Transition {path=AnimGraph.AnimGraphNode_StateMachine_1.Gliding.AnimStateTransitionNode_6.Transition, nodes [Result(bCanEnterTransition=true)]}
 Transition {
  path=AnimGraph.AnimGraphNode_StateMachine_1.Gliding.AnimStateTransitionNode_7.Transition
  nodes [Result(bCanEnterTransition=Get bIsSkydivingDiving())]
 }
 Transition {path=AnimGraph.AnimGraphNode_StateMachine_1.Gliding.AnimStateTransitionNode_8.Transition, nodes [Result(bCanEnterTransition=true)]}
 DBNO {
  entry=Locomotion
  path=AnimGraph.AnimGraphNode_StateMachine_2.DBNO
  ids=id
  nodes [
   AnimStateNode_0 = Locomotion()
   AnimStateNode_1 = DBNOInto() {is=title AnimStateNode}
   AnimStateNode_2 = DBNOIdle() {is=title AnimStateNode}
   AnimStateNode_3 = DBNOOutro() {is=title AnimStateNode}
   AnimStateTransitionNode_0 = AnimStateTransitionNode()
   AnimStateTransitionNode_1 = AnimStateTransitionNode() {bAutomaticRuleBasedOnSequencePlayerInState=true, crossfadeDuration=0.6}
   AnimStateTransitionNode_3 = AnimStateTransitionNode() {crossfadeDuration=0.5}
   AnimStateTransitionNode_4 = AnimStateTransitionNode() {bAutomaticRuleBasedOnSequencePlayerInState=true, crossfadeDuration=0.4}
  ]
  links [
   {
    dst=DBNOOutro
    rule=AnimGraph.AnimGraphNode_StateMachine_2.DBNO.AnimStateTransitionNode_3.Transition
    src=DBNOIdle
    via=AnimStateTransitionNode_3
   }
   {
    dst=DBNOIdle
    rule=AnimGraph.AnimGraphNode_StateMachine_2.DBNO.AnimStateTransitionNode_1.Transition
    src=DBNOInto
    via=AnimStateTransitionNode_1
   }
   {
    dst=Locomotion
    rule=AnimGraph.AnimGraphNode_StateMachine_2.DBNO.AnimStateTransitionNode_4.Transition
    src=DBNOOutro
    via=AnimStateTransitionNode_4
   }
   {
    dst=DBNOInto
    rule=AnimGraph.AnimGraphNode_StateMachine_2.DBNO.AnimStateTransitionNode_0.Transition
    src=Locomotion
    via=AnimStateTransitionNode_0
   }
  ]
 }
 Locomotion {
  path=AnimGraph.AnimGraphNode_StateMachine_2.DBNO.AnimStateNode_0.Locomotion
  nodes [Output Animation Pose(Result=Use cached pose 'LocoCache'())]
 }
 DBNOInto {
  path=AnimGraph.AnimGraphNode_StateMachine_2.DBNO.AnimStateNode_1.DBNOInto
  nodes [Output Animation Pose(Result=Sequence Player 'Wayfare_Mask_DBNO_Into_PET'() {sequence=/BRCosmetics/Accessories/FORT_Backpacks/Backpack_F_MED_Wayfare_Mask/Animations/Wayfare_Mask_DBNO_Into_PET})]
 }
 DBNOIdle {
  path=AnimGraph.AnimGraphNode_StateMachine_2.DBNO.AnimStateNode_2.DBNOIdle
  nodes [Output Animation Pose(Result=Sequence Player 'Wayfare_Mask_DBNO_Loop_PET'() {sequence=/BRCosmetics/Accessories/FORT_Backpacks/Backpack_F_MED_Wayfare_Mask/Animations/Wayfare_Mask_DBNO_Loop_PET})]
 }
 DBNOOutro {path=AnimGraph.AnimGraphNode_StateMachine_2.DBNO.AnimStateNode_3.DBNOOutro, nodes [Output Animation Pose()]}
 Transition {path=AnimGraph.AnimGraphNode_StateMachine_2.DBNO.AnimStateTransitionNode_0.Transition, nodes [Result(bCanEnterTransition=Get bIsDBNO())]}
 Transition {path=AnimGraph.AnimGraphNode_StateMachine_2.DBNO.AnimStateTransitionNode_1.Transition, nodes [Result(bCanEnterTransition=true)]}
 Transition {
  path=AnimGraph.AnimGraphNode_StateMachine_2.DBNO.AnimStateTransitionNode_3.Transition
  nodes [Result(bCanEnterTransition=NOT Boolean(A=Get bIsDBNO()))]
 }
 Transition {path=AnimGraph.AnimGraphNode_StateMachine_2.DBNO.AnimStateTransitionNode_4.Transition, nodes [Result(bCanEnterTransition=true)]}
 EventGraph {nodes [
   AnimNotify_Entered_IntroRun() {is=title K2Node_Event}
   > Set AllowDirectionChange(AllowDirectionChange:bool=false) -> (Output_Get:bool)
   > Set IntroRunStarted(IntroRunStarted:bool=true) -> (Output_Get:bool)
   > Set Locomotion Cardinal Direction When Run Started(Locomotion Cardinal Direction When Run Started:EFortCardinalDirection Enum=Get LocomotionCardinalDirection(self=Get As Fortnite M Avg Player Anim Blueprint()) -> (LocomotionCardinalDirection=North)) -> (Output_Get:EFortCardinalDirection Enum=North)
   > Delay(Duration=5)
   > Set IntroRunStarted(IntroRunStarted=false)
   > Set Locomotion Cardinal Direction When Run Started(Locomotion Cardinal Direction When Run Started=Get LocomotionCardinalDirection(self=Get As Fortnite M Avg Player Anim Blueprint()) -> (LocomotionCardinalDirection=North)) -> (Output_Get=North)
   Event BlueprintInitializeAnimation()
   > n1 = Cast To PlayerPawn_Athena(Object=TryGetPawnOwner())
   > n2 = Set As Player Pawn Athena(As Player Pawn Athena:Player Pawn Athena Object Reference=@n1) -> (Output_Get:Player Pawn Athena Object Reference)
   > n3 = Cast To Fortnite_M_Avg_Player_AnimBlueprint!(execute, Object=GetAnimInstance(self=Get Mesh(self=@n2))) -> (then, CastFailed, AsFortnite M Avg Player Anim Blueprint:Fortnite M Avg Player Anim Blueprint Object Reference) {is=title K2Node_DynamicCast}
   > Set As Fortnite M Avg Player Anim Blueprint(As Fortnite M Avg Player Anim Blueprint:Fortnite M Avg Player Anim Blueprint Object Reference=@n3) -> (Output_Get:Fortnite M Avg Player Anim Blueprint Object Reference)
   Event BlueprintUpdateAnimation()
   > n4 = Branch(Condition=Get IntroRunStarted() -> (IntroRunStarted:bool))
   n5 = Get As Fortnite M Avg Player Anim Blueprint()
   Branch(
    execute=@n4.else
    Condition="Not Equal (Enum)"!(
     A:EFortCardinalDirection Enum=Get LocomotionCardinalDirection(self=@n5) -> (LocomotionCardinalDirection=North)
     B:EFortCardinalDirection Enum=Get Locomotion Cardinal Direction When Run Started() -> (Locomotion Cardinal Direction When Run Started:EFortCardinalDirection Enum=North)
    ) -> (ReturnValue:bool)
   )
   > Set AllowDirectionChange(AllowDirectionChange=true)
   AnimNotify_Entered_JumpOutroRun() {is=title K2Node_Event}
   > Set JustLandedFromJump(JustLandedFromJump:bool=true) -> (Output_Get:bool)
   > n6 = RetriggerableDelay(Duration=2)
   Set JustLandedFromJump(execute=[@n6.then, @n7.then], JustLandedFromJump=false)
   n7 = AnimNotify_Entered_OutroRun() {is=title K2Node_Event}
   AnimNotify_Entered_JumpOutro() {is=title K2Node_Event}
  ]}
]
variables [
 AllowDirectionChange {type=boolean}
 As Fortnite M Avg Player Anim Blueprint {type=Fortnite_M_Avg_Player_AnimBlueprint_C, default=null}
 As Player Pawn Athena {type=PlayerPawn_Athena_C}
 fin pins 01 {type=Rotator}
 IntroRunStarted {type=boolean}
 JustLandedFromJump {type=boolean}
 Locomotion Cardinal Direction When Run Started {type=EFortCardinalDirection, default=North}
]
